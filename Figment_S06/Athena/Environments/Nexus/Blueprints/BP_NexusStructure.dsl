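; BP_NexusStructure is a Nexus environment Blueprint for a structure with optional glow-point and spline components.
; Its construction script runs a sequence that destroys SM_Nexus_GlowPoints when DisableGlowPoints is set, then SM_Nexus_Splines when DisableSplines is set.
; The BeginPlay, Tick and ActorBeginOverlap events are empty, so the only logic runs at construction.
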(event EventBeginPlay)

(event Collision|EventActorBeginOverlap
  (params (OtherActor :type Actor%20Object%20Reference)))

(event EventTick
  (params (DeltaSeconds :type Float%20%28single-precision%29)))

(fn ConstructionScript
  (Utilities|FlowControl|Sequence
    (:then_0
      (if (Variables|Default|GetDisableGlowPoints)
        (Components|DestroyComponent (Variables|Default|GetSM_Nexus_GlowPoints))))
    (:then_1
      (if (Variables|Default|GetDisableSplines)
        (Components|DestroyComponent (Variables|Default|GetSM_Nexus_Splines))))))
