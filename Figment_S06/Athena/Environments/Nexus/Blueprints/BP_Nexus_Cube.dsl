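; BP_Nexus_Cube is an actor in the Nexus environment that animates a cube's height to make it float.
; On BeginPlay it calls SetCubeFloat, which stores the current location in Origin_X and plays a timeline from the start.
; Each timeline update adds the Y component of a vector track to the stored Z and sets the location without sweep, keeping X and Y at the origin.
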(event EventBeginPlay
  (CallFunction|SetCubeFloat))

(event Custom|SetCubeFloat
  (bind (_returnvalue_x _returnvalue_y _returnvalue_z) (Math|Vector|BreakVector (Transformation|GetActorLocation)))
  (bind (_origin_x _origin_y _origin_z) (Math|Vector|BreakVector (Variables|GetOrigin_X)))
  (Variables|SetOrigin_X (Math|Vector|MakeVector _returnvalue_x _returnvalue_y _returnvalue_z))
  (bind (_direction _newtrack_0) (Timeline|Timeline PlayFromStart)
    (:Update
      (bind (_newtrack_0_x _newtrack_0_y) (Math|Vector|BreakVector _newtrack_0))
      (Transformation|SetActorLocation 0 (Math|Vector|MakeVector _origin_x _origin_y (+ _origin_z _newtrack_0_y)) false false))))

(fn ConstructionScript)
