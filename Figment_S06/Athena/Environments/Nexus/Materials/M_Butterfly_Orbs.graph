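# Translucent unlit material for the butterfly orbs in the Nexus environment, with a tintable bright cyan color.
# A Fresnel function with brightness and exponent parameters feeds a depth fade to shape soft opacity.
# World position offset pushes each orb along the normalized direction from its object position toward a hardcoded world point, scaled by MoveOrbs from the MPC_Cube parameter collection.
# That offset passes through MF_SpawnInBetween, so movement is driven globally rather than per material.
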Material /Figment_S06/Athena/Environments/Nexus/Materials/M_Butterfly_Orbs
graph {
 properties {blendMode=BLEND_Translucent, pixelDepthOffsetMode=PDOM_Legacy, shadingModel=MSM_Unlit, usedShadingModels=Unlit}
 nodes [
  DepthFade(
   Opacity=MaterialFunctionCall(
    Brightness=ScalarParameter() {defaultValue=1, parameterName=Fresnel Brightness}
    Exponent=ScalarParameter() {defaultValue=10, parameterName=Fresnel Exponent}
   ) {materialFunction=&MF_Fresnel_FX}.Fresnel
   FadeDistance=ScalarParameter() {defaultValue=3000, parameterName=DepthFade}
  ) {fadeDistanceDefault=1000}
  n1 = Subtract(A=Constant3Vector() {constant=<rgba -713 2456 244652 0>}, B=ObjectPositionWS())
  MaterialFunctionCall!(WPO=Multiply(
    A=Normalize(VectorInput=@n1)
    B=CollectionParameter() {collection=/Game/Athena/Prototype/Blueprints/Cube/Materials/MPC_Cube, parameterName=MoveOrbs}
   )) {materialFunction=/Game/Athena/Environments/Nexus/Materials/MF_SpawnInBetween}
  VectorParameter() {parameterName=Color, defaultValue=<rgba 1.77 4.031 5 1>}
 ]
}
