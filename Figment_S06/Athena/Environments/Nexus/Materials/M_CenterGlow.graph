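# Additive, unlit, two-sided glow material for the Nexus environment.
# A radial UV falloff, shaped by a power and tinted by Color, gets depth fade on the default shading path but not on mobile.
# A separate branch blends two colors by a cross-flare texture's green channel and feeds DeriveHDRfromLDR.
# MF_SpawnInBetween supplies a world position offset function.
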Material /Figment_S06/Athena/Environments/Nexus/Materials/M_CenterGlow
graph {
 properties {blendMode=BLEND_Additive, pixelDepthOffsetMode=PDOM_Legacy, shadingModel=MSM_Unlit, twoSided=true, usedShadingModels=Unlit}
 nodes [
  MaterialFunctionCall!(WPO) {materialFunction=/Game/Athena/Environments/Nexus/Materials/MF_SpawnInBetween}
  n1 = LinearInterpolate(
   A=VectorParameter() {parameterName=Color0}.RGB
   B=VectorParameter() {parameterName=Color1, defaultValue=<rgba 7.9922 11.7611 12 0>}.RGB
   Alpha=TextureSample() {texture=/Game/Effects/Fort_Effects/Textures/Flares/T_Cross_Flare_ColorsRGB}.G
  )
  MaterialFunctionCall(LDR Input=Clamp(None=@n1), HDR Tint=@n1) {materialFunction=&DeriveHDRfromLDR}
  n2 = Clamp(None=Distance(A=TextureCoordinate(), B=Constant2Vector() {g=0.5, r=0.5}))
  n3 = Multiply(
   A=VectorParameter() {parameterName=Color, defaultValue=<rgba 0.8776 0 10 1>}.RGB
   B=Power(Base=OneMinus(None=@n2), Exp=ScalarParameter() {defaultValue=5.1026, parameterName=Falloff})
  )
  ShadingPathSwitch(Default=DepthFade(Opacity=@n3, FadeDistance=ScalarParameter() {defaultValue=14713.5859, parameterName=DepthFade}), Mobile=@n3) {desc=Migrated from FeatureLevelSwitch}
 ]
}
