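# M_DistantGround is an unlit material for distant ground in the Nexus environment.
# A color gradient follows distance from the actor, scaled by object size, plus glow from static and panning glitch noise.
# Noise-modulated Fresnel serves as UVs into a chromatic smoke texture for iridescence, and a UV1 edge gradient tints the sum.
# The graph also holds a camera-relative parallax offset and a MF_SpawnInBetween world position offset function.
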Material /Figment_S06/Athena/Environments/Nexus/Materials/M_DistantGround
graph {
 properties {pixelDepthOffsetMode=PDOM_Legacy, shadingModel=MSM_Unlit, usedShadingModels=Unlit}
 nodes [
  n1 = VectorParameter() {parameterName=WorldCenter, defaultValue=<rgba 0 0 305948.6562 1>}
  n2 = WorldPosition()
  DotProduct(A=Normalize(VectorInput=Subtract(A=@n1.RGB, B=@n2.XYZ)), B=VertexNormalWS())
  n3 = TextureSample(UVs=Multiply(A=TextureCoordinate(), B=ScalarParameter() {defaultValue=35.9415, parameterName=Tiling})) {samplerType=SAMPLERTYPE_LinearGrayscale, texture=/Game/Global/Textures/Default/Noise/glitch_noise}
  Fresnel(ExponentIn=@n3.R) {baseReflectFraction=0}
  MaterialFunctionCall!(WPO) {materialFunction=/Game/Athena/Environments/Nexus/Materials/MF_SpawnInBetween}
  n4 = WorldPosition()
  n5 = ActorPositionWS()
  n6 = ScalarParameter() {defaultValue=1467.8466, parameterName=DistanceBlend}
  n7 = MaterialFunctionCall() {materialFunction=/Engine/Functions/Engine_MaterialFunctions02/WorldPositionOffset/ObjectScale}
  n8 = Clamp(None=Divide(A=Distance(A=@n4.XYZ, B=@n5), B=Multiply(A=@n6, B=@n7.Scale Z)))
  Multiply(
   A=Subtract(A=CameraPositionWS(), B=ActorPositionWS())
   B=LinearInterpolate(
    A=ScalarParameter() {defaultValue=0.4762, parameterName=Paralaxing}
    B=ScalarParameter() {defaultValue=0.7714, parameterName=Paralaxing_1}
    Alpha=@n8
   )
  )
  n9 = TextureCoordinate()
  n10 = ScalarParameter() {defaultValue=7.2564, parameterName=Tiling_1}
  Multiply(
   A=TextureSample(UVs=Multiply(A=@n9, B=@n10)) {samplerType=SAMPLERTYPE_LinearColor, texture=/Game/Packages/Fortress_Sky/Materials/Noise32x322_nearestNeighboor}.R
   B=@n3.R
  )
  n11 = Fresnel(ExponentIn=LinearInterpolate(
    A=ScalarParameter() {defaultValue=1.1452, parameterName=ChromaticEXP1}
    B=ScalarParameter() {defaultValue=26.4786, parameterName=ChromaticEXP2}
    Alpha=@n3.R
   )) {baseReflectFraction=0}
  n12 = Panner(
   Coordinate=Multiply(A=TextureCoordinate(), B=ScalarParameter() {defaultValue=11.4743, parameterName=Tiling_2})
   Speed=AppendVector(A=Constant(), B=ScalarParameter() {defaultValue=0.01, parameterName=PanSpeed})
  ) {bFractionalPart=true}
  n13 = Multiply(
   A=VectorParameter() {parameterName=Glow, defaultValue=<rgba 0.474 0.474 0.474 1>}.RGB
   B=Multiply(A=@n3.R, B=TextureSample(UVs=@n12) {samplerType=SAMPLERTYPE_LinearGrayscale, texture=&glitch_noise}.R)
  )
  n14 = LinearInterpolate(
   A=VectorParameter() {parameterName=Color, defaultValue=<rgba 0 0.2074 0.5625 1>}.RGB
   B=VectorParameter() {parameterName=Color_1, defaultValue=<rgba 0.4638 0.5477 0.7188 1>}.RGB
   Alpha=@n8
  )
  n15 = ComponentMask(None=TextureCoordinate() {coordinateIndex=1}) {r=true}
  n16 = Multiply(A=Multiply(A=@n15, B=OneMinus(None=@n15))) {constB=2}
  n17 = ScalarParameter() {defaultValue=0.3411, parameterName=EXPEdge}
  Multiply(
   A=Add(A=TextureSample(UVs=@n11) {texture=/Game/Effects/Fort_Effects/Textures/Tile/T_Inky_Smoke_Tile_Chromatic}.RGB, B=Add(A=@n13, B=@n14))
   B=LinearInterpolate(
    A=VectorParameter() {parameterName=ColorEdge, defaultValue=<rgba 0.5 0.5 0.5 1>}.RGB
    B=VectorParameter() {parameterName=ColorEdge_1, defaultValue=<rgba 2 2 2 1>}.RGB
    Alpha=Power(Base=@n16, Exp=@n17)
   )
  )
  ScalarParameter() {defaultValue=1.0389, parameterName=CromaticExponent}
  WorldPosition()
 ]
}
