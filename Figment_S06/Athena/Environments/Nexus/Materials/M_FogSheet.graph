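# M_FogSheet is an unlit, two-sided translucent material for fog sheets in the Nexus environment.
# Two independently scaled, panning UV sets sample cloud height, normal and cloudscape textures.
# Cloudscape opacity gets a height-modulated depth fade, skipped on mobile, then alpha scaling and an optional static-switch radial fade.
# View-angle fade, near-camera fade, blended normals and base color nodes also exist, with no material output links shown.
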Material /Figment_S06/Athena/Environments/Nexus/Materials/M_FogSheet
graph {
 properties {
  bUseTranslucencyVertexFog=false
  blendMode=BLEND_Translucent
  pixelDepthOffsetMode=PDOM_Legacy
  shadingModel=MSM_Unlit
  translucencyLightingMode=TLM_VolumetricPerVertexDirectional
  twoSided=true
  usedShadingModels=Unlit
 }
 nodes [
  n1 = TextureCoordinate()
  n2 = Multiply(A=ScalarParameter() {defaultValue=6, parameterName=BaseTexture1_ScaleX}, B=ComponentMask(None=@n1) {r=true})
  n3 = Multiply(A=ComponentMask(None=@n1) {g=true}, B=ScalarParameter() {defaultValue=6, parameterName=BaseTexture1_ScaleY})
  n4 = Panner(
   Coordinate=Add(A=AppendVector(A=@n2, B=@n3))
   Speed=ComponentMask(None=VectorParameter() {parameterName=BaseTexture_Speed, defaultValue=<rgba 0 0.012 0 0>}.RGB) {g=true, r=true}
  ) {speedY=-5}
  n5 = TextureSample(UVs=@n4) {samplerType=SAMPLERTYPE_LinearColor, texture=/Figment_S06/Packages/DS_Fortnite_Terrain_NoLOD/Textures/Cloud_Height4}
  n6 = TextureCoordinate()
  n7 = Multiply(A=ScalarParameter() {defaultValue=3, parameterName=BaseTexture1_ScaleX_1}, B=ComponentMask(None=@n6) {r=true})
  n8 = Multiply(A=ComponentMask(None=@n6) {g=true}, B=ScalarParameter() {defaultValue=3, parameterName=BaseTexture1_ScaleY_1})
  n9 = Panner(
   Coordinate=Add(A=AppendVector(A=@n7, B=@n8))
   Speed=ComponentMask(None=VectorParameter() {parameterName=BaseTexture_Speed_1, defaultValue=<rgba 0 0.012 0 0>}.RGB) {g=true, r=true}
  ) {speedY=-5}
  LinearInterpolate(
   A=@n5.G
   B=TextureSample(UVs=@n9) {samplerType=SAMPLERTYPE_LinearColor, texture=&Cloud_Height4}.G
  )
  MaterialFunctionCall(
   Normal=MaterialFunctionCall(
    BaseNormal=TextureSample(UVs=@n4) {samplerType=SAMPLERTYPE_Normal, texture=/Game/Packages/DS_Fortnite_Terrain_NoLOD/Textures/Cloud_Normal4}.RGB
    AdditionalNormal=TextureSample(UVs=@n9) {samplerType=SAMPLERTYPE_Normal, texture=&Cloud_Normal4}.RGB
   ) {materialFunction=&BlendAngleCorrectedNormals}.Result
   Flatness=ScalarParameter() {parameterName=FlattenNormal}
  ) {materialFunction=&FlattenNormal}
  n10 = TextureSample(UVs=@n4) {texture=/Game/Effects/Fort_Effects/Textures/Tile/T_CloudScape_08}
  n11 = ScalarParameter() {defaultValue=600, parameterName=FadeDistance}
  n12 = ShadingPathSwitch(Default=DepthFade(Opacity=@n10.G, FadeDistance=LinearInterpolate(A=@n11, Alpha=@n5.G) {constA=600, constB=1000}), Mobile=@n10.G) {desc=Migrated from FeatureLevelSwitch}
  Multiply(
   A=MaterialFunctionCall("FadeSensitivity (optional)"=ScalarParameter() {defaultValue=1, parameterName=ViewAngleFade}) {materialFunction=&FadeBasedOnViewAngle}.Result
   B=@n12
  )
  n13 = Clamp(None=Multiply(A=ScalarParameter() {defaultValue=0.6, parameterName=ScaleAlpha}, B=@n12))
  Multiply(A=MaterialFunctionCall() {materialFunction=&Near_Camera_Fade}.Result, B=@n13)
  n14 = TextureCoordinate()
  n15 = Constant2Vector() {g=0.5, r=0.5}
  n16 = LinearInterpolate(
   A=ScalarParameter() {parameterName=FadeRadius_Contrast_1}
   B=ScalarParameter() {defaultValue=1, parameterName=FadeRadius_Contrast_2}
   Alpha=Divide(A=Distance(A=@n14, B=@n15), B=ScalarParameter() {defaultValue=1, parameterName=FadeRadius})
  )
  StaticSwitchParameter(True=Multiply(A=@n13, B=Clamp(None=@n16)), False=@n13) {parameterName=UseRadialFade}
  TextureCoordinate()
  TextureSample(UVs=@n9) {texture=&T_CloudScape_08}
  VectorParameter() {parameterName=BaseColor, defaultValue=<rgba 1 0.3785 0.7883 0>}
 ]
}
