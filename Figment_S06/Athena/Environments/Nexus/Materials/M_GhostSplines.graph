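# Additive, unlit, two-sided material for glowing, animated ghost spline strands in the Nexus environment.
# SplineThicken widens the strands by a Width parameter, summed with a camera-vector push for a world position offset.
# A time-scrolled, tiled storm-cloud tube mask is multiplied by a U-based falloff that fades both edges, then raised to a power.
# That mask feeds a brightness multiplier and a refraction lerp, alongside a blue Color parameter.
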Material /Figment_S06/Athena/Environments/Nexus/Materials/M_GhostSplines
graph {
 properties {
  blendMode=BLEND_Additive
  pixelDepthOffsetMode=PDOM_Legacy
  shadingModel=MSM_Unlit
  translucencyPass=MTP_BeforeDOF
  twoSided=true
  usedShadingModels=Unlit
 }
 nodes [
  n1 = ScalarParameter() {defaultValue=1681.2651, parameterName=Width}
  Add(
   A=MaterialFunctionCall(WidthBase=@n1, WidthTip=@n1) {materialFunction=&SplineThicken}.WorldPositionOffset
   B=Multiply(
    A=MaterialFunctionCall!(Use World Position Offsets) {materialFunction=/Engine/Functions/Engine_MaterialFunctions02/Utility/CameraVectorWithWPOOptions}.Result
    B=ScalarParameter() {defaultValue=3440.8789, parameterName=CameraPush}
   )
  )
  CameraVectorWS()
  n2 = Multiply(A=Time(), B=ScalarParameter() {defaultValue=-0.001, parameterName=Speed})
  n3 = Multiply(
   A=Add(A=TextureCoordinate(), B=Frac(None=@n2))
   B=AppendVector(A=ScalarParameter() {defaultValue=0.4111, parameterName=Tiling_X}, B=ScalarParameter() {defaultValue=17.5422, parameterName=Tiling_Y})
  )
  n4 = ComponentMask(None=TextureCoordinate()) {r=true}
  n5 = OneMinus(None=@n4)
  n6 = Power(
   Base=Multiply(
    A=TextureSample(UVs=@n3) {samplerType=SAMPLERTYPE_Masks, texture=/Game/Athena/SafeZone/Textures/StormCloudtubes_M}.R
    B=Multiply(A=@n4, B=@n5)
   )
   Exp=ScalarParameter() {defaultValue=2.2, parameterName=Exp}
  )
  LinearInterpolate(A=ScalarParameter() {defaultValue=1, parameterName=Refract}, B=ScalarParameter() {defaultValue=1.7805, parameterName=Refract_1}, Alpha=@n6)
  Multiply(A=ScalarParameter() {defaultValue=9.6857, parameterName=Birghtness}, B=@n6)
  VectorParameter() {parameterName=Color, defaultValue=<rgba 0.1823 0.3672 1 1>}
 ]
}
