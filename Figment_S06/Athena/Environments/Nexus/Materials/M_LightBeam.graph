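# Additive, unlit, two-sided material for a tinted light beam in the Nexus environment.
# SplineThicken uses a shared Width parameter to produce world position offset, which feeds MF_SpawnInBetween.
# A radial UV gradient from the center feeds two color treatments: DeriveHDRfromLDR with Color and Range, and a Falloff power curve multiplied by Color.
# A soft flare texture sample is present but feeds nothing.
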Material /Figment_S06/Athena/Environments/Nexus/Materials/M_LightBeam
graph {
 properties {blendMode=BLEND_Additive, pixelDepthOffsetMode=PDOM_Legacy, shadingModel=MSM_Unlit, twoSided=true, usedShadingModels=Unlit}
 nodes [
  n1 = TextureCoordinate()
  n2 = Constant2Vector() {g=0.5, r=0.5}
  n3 = OneMinus(None=Clamp(None=Distance(A=@n1, B=@n2)))
  n4 = VectorParameter() {parameterName=Color, defaultValue=<rgba 1.0173 1.4758 2 1>}
  MaterialFunctionCall(LDR Input=@n3, HDR Tint=@n4.RGB, Derive HDR Power=ScalarParameter() {defaultValue=35.7814, parameterName=Range}) {materialFunction=&DeriveHDRfromLDR}
  n5 = ScalarParameter() {defaultValue=15275.9023, parameterName=Width}
  MaterialFunctionCall!(WPO=MaterialFunctionCall(WidthBase=@n5, WidthTip=@n5) {materialFunction=&SplineThicken}.WorldPositionOffset) {materialFunction=/Game/Athena/Environments/Nexus/Materials/MF_SpawnInBetween}
  Multiply(A=@n4.RGB, B=Power(Base=@n3, Exp=ScalarParameter() {defaultValue=25.8472, parameterName=Falloff}))
  TextureSample() {texture=/Game/Effects/Fort_Effects/Textures/Flares/T_Flare_Soft}
 ]
}
