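# Translucent unlit material for the Nexus environment that, going by its name and graph, builds an animated maelstrom swirl from a storm-cloud mask.
# World XY becomes radial coordinates, with powered radial falloff and an X-into-Y shear twisting them into a spiral that pans over time.
# The mask's green channel is sampled with mesh-UV derivatives for mip selection, then feeds a depth fade.
# A color parameter and exposed scale, twist, speed and fade parameters allow tuning.
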Material /Figment_S06/Athena/Environments/Nexus/Materials/M_Nexus_Malestrom
graph {
 properties {blendMode=BLEND_Translucent, pixelDepthOffsetMode=PDOM_Legacy, shadingModel=MSM_Unlit, usedShadingModels=Unlit}
 nodes [
  n1 = WorldPosition() {worldPositionShaderOffset=WPT_ExcludeAllShaderOffsets}
  n2 = MaterialFunctionCall(Vector or UVs=Divide(A=ComponentMask(None=@n1.XYZ) {g=true, r=true}, B=ScalarParameter() {defaultValue=2504.9761, parameterName=Scale_1})) {materialFunction=&VectorToRadialValue}
  n3 = Power(Base=ComponentMask(None=@n2.Radial Coordinates) {g=true}, Exp=ScalarParameter() {defaultValue=0.6162, parameterName=FallofTwist})
  Clamp(None=OneMinus(None=@n3))
  n4 = Constant() {r=1}
  n5 = Multiply(
   A=Add(A=@n3, B=ScalarParameter() {defaultValue=-0.224, parameterName=Scale_Twist_Bias})
   B=ScalarParameter() {defaultValue=1.0662, parameterName=Scale_Twist_Y}
  )
  n6 = Multiply(A=@n2.Radial Coordinates, B=AppendVector(A=ScalarParameter() {defaultValue=5, parameterName=Scale_Twist_X}, B=Divide(A=@n4, B=@n5)))
  n7 = Add(A=@n6, B=AppendVector(A=Constant(), B=ComponentMask(None=@n6) {r=true}))
  n8 = ScalarParameter() {defaultValue=-0.21, parameterName=Speed_x}
  n9 = ScalarParameter() {defaultValue=0.1, parameterName=Speed_y}
  n10 = Frac(None=Multiply(A=Time(), B=AppendVector(A=@n8, B=@n9)))
  n11 = TextureCoordinate()
  DepthFade(
   Opacity=TextureSample!(UVs=Add(A=@n7, B=@n10), Tex, "DDX(UVs)"=DDX(Value=@n11), "DDY(UVs)"=DDY(Value=@n11), Apply View MipBias) {mipValueMode=TMVM_Derivative, samplerType=SAMPLERTYPE_Masks, texture=/Game/Athena/SafeZone/Textures/StormCloudtubes_M}.G
   FadeDistance=ScalarParameter() {defaultValue=188126.9688, parameterName=Depthfade}
  )
  VectorParameter() {parameterName=Color, defaultValue=<rgba 0.3698 0.637 1 1>}
 ]
}
