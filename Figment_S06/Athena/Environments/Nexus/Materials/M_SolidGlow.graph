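# M_SolidGlow is an unlit material in the Nexus environment set, built for a solid glowing look.
# It defines a Color1 vector parameter with a bright, over-one cyan-white default for HDR glow.
# It calls the MF_SpawnInBetween material function in a node labeled WPO, suggesting shared spawn-related vertex offset logic.
# The graph shows no textures or other shading logic.
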Material /Figment_S06/Athena/Environments/Nexus/Materials/M_SolidGlow
graph {
 properties {pixelDepthOffsetMode=PDOM_Legacy, shadingModel=MSM_Unlit, usedShadingModels=Unlit}
 nodes [
  MaterialFunctionCall!(WPO) {materialFunction=/Game/Athena/Environments/Nexus/Materials/MF_SpawnInBetween}
  VectorParameter() {parameterName=Color1, defaultValue=<rgba 7.9922 11.7611 12 0>}
 ]
}
