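# M_StarField is an unlit additive material for the butterfly star field in the Nexus environment.
# It sums a flow-animated layer and a bump-offset layer of one star texture, each separately tiled, then tints the result.
# World position offset scales the camera-to-actor vector through MF_SpawnInBetween for parallax.
# It adds a push term driven by camera distance beyond a threshold and weighted by a cubed view-alignment dot product.
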Material /Figment_S06/Athena/Environments/Nexus/Materials/M_StarField
graph {
 properties {blendMode=BLEND_Additive, pixelDepthOffsetMode=PDOM_Legacy, shadingModel=MSM_Unlit, usedShadingModels=Unlit}
 nodes [
  n1 = AppendVector(A=Constant(), B=ScalarParameter() {defaultValue=0.025, parameterName=OffsetFlow})
  n2 = ScalarParameter() {defaultValue=0.2855, parameterName=Speed}
  n3 = Time()
  n4 = Multiply(
   A=TextureCoordinate()
   B=AppendVector(A=ScalarParameter() {defaultValue=8, parameterName=Tiling_X}, B=ScalarParameter() {defaultValue=1, parameterName=Tiling_Y})
  )
  n5 = ComponentMask(None=MaterialFunctionCall!(
    Diffuse=TextureObject() {texture=/Figment_S06/Athena/Environments/Nexus/Textures/T_ButterflyStarField_D}
    Normal Map
    "Flow Vector Map (see tooltip)"=@n1
    UV texture 2 offset=@n1
    "Time (see tooltip)"=Multiply(A=@n2, B=@n3)
    UVs=@n4
   ) {materialFunction=/Engine/Functions/Engine_MaterialFunctions02/Texturing/FlowMaps_Simple}.Diffuse) {r=true}
  n6 = BumpOffset(Coordinate=TextureCoordinate(), Height=ScalarParameter() {defaultValue=-1, parameterName=Height})
  n7 = AppendVector(
   A=ScalarParameter() {defaultValue=9.3937, parameterName=Tiling_X_Offset}
   B=ScalarParameter() {defaultValue=-0.2714, parameterName=Tiling_Y_Offset}
  )
  n8 = ScalarParameter() {defaultValue=9, parameterName=Tiling_X_1}
  n9 = ScalarParameter() {defaultValue=2.3077, parameterName=Tiling_Y_1}
  n10 = TextureSample(UVs=Multiply(A=Add(A=@n6, B=@n7), B=AppendVector(A=@n8, B=@n9))) {texture=&T_ButterflyStarField_D}
  Add(A=Multiply(A=VectorParameter() {parameterName=Color, defaultValue=<rgba 0.3039 1.2135 2 1>}.RGB, B=Add(A=@n5, B=@n10.R)))
  n11 = Subtract(A=CameraPositionWS(), B=ActorPositionWS())
  n12 = MaterialFunctionCall!(WPO=Multiply(A=@n11, B=ScalarParameter() {defaultValue=0.7, parameterName=Paralaxing})) {materialFunction=/Game/Athena/Environments/Nexus/Materials/MF_SpawnInBetween}
  n13 = NamedRerouteUsage()
  n14 = Length(None=@n11) {is=class MaterialExpressionLength}
  n15 = ScalarParameter() {parameterName=PushBackDistanceThreshold}
  n16 = Multiply(A=Max(A=Subtract(A=@n14, B=@n15) {constB=110000}) {constB=0})
  n17 = Subtract(
   A=Add(A=WorldPosition() {worldPositionShaderOffset=WPT_ExcludeAllShaderOffsets}.XYZ, B=NamedRerouteUsage()) {desc=after offset}
   B=CameraPositionWS()
  ) {desc=camera vector WS}
  n18 = Normalize(VectorInput=@n17)
  n19 = Normalize(VectorInput=Multiply(A=NamedRerouteUsage()) {constB=-1})
  n20 = Power(Base=Saturate(None=DotProduct(A=@n18, B=@n19))) {constExponent=3}
  Add(A=@n12.Result, B=Multiply(A=Normalize(VectorInput=@n13), B=Multiply(A=@n16, B=@n20)) {constB=150000})
  DepthFade() {fadeDistanceDefault=150000}
  NamedRerouteDeclaration(None=@n11) {name=ActorToCam, nodeColor=<rgba 0.3594 1 0 1>}
  NamedRerouteDeclaration(None=@n12.Result) {name=offset, nodeColor=<rgba 0.1719 0 1 1>}
  NamedRerouteDeclaration(None=@n17) {name=Camera Vector WS, nodeColor=<rgba 0 0.7891 1 1>}
  TextureSample(UVs=@n4) {texture=&T_ButterflyStarField_D}
 ]
}
