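# M_Whiteout is an unlit material in the Nexus environment set, named for a whiteout effect.
# Its graph holds a single vector parameter whose default is an overbright, HDR white.
# No texture, math or other nodes are present, and the graph shows no link from the parameter to a material output.
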Material /Figment_S06/Athena/Environments/Nexus/Materials/M_Whiteout
graph {
 properties {pixelDepthOffsetMode=PDOM_Legacy, shadingModel=MSM_Unlit, usedShadingModels=Unlit}
 nodes [VectorParameter() {defaultValue=<rgba 2 2 2 1>}]
}
