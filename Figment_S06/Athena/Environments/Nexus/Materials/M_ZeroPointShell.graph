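# Unlit material for the Nexus Zero Point shell, layering storm clouds, falloff glow and butterfly backgrounds.
# Storm-cloud masks project onto a virtual plane, warped by scrolling noise, with one branch adding vertical and distance falloff tints.
# Another branch adds two butterfly samples, one bump-offset for parallax, and an HDR gradient warped by panning normal maps.
# The graph also holds radial twist, camera parallax and MF_SpawnInBetween nodes, without visible output connections.
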Material /Figment_S06/Athena/Environments/Nexus/Materials/M_ZeroPointShell
graph {
 properties {pixelDepthOffsetMode=PDOM_Legacy, shadingModel=MSM_Unlit, usedShadingModels=Unlit}
 nodes [
  n1 = Constant()
  n2 = MaterialFunctionCall(
   Plane Normal Axis=Constant3Vector() {constant=<rgba 0 0 1 0>}
   Plane X Axis=Constant3Vector() {constant=<rgba 1 0.0015 0 1>}
   Plane Center=MaterialFunctionCall(X=@n1, Y=@n1, Z=ScalarParameter() {defaultValue=7904700.5, parameterName=Planez}) {materialFunction=&MakeFloat3}.Result
   UV Scale=ScalarParameter() {defaultValue=48162788, parameterName=VirtualPlaneScale}
  ) {materialFunction=&VirtualPlaneCoordinates}
  n3 = Multiply(A=@n2.UVs, B=Constant2Vector() {g=-1, r=1})
  n4 = TextureCoordinate()
  n5 = Constant()
  n6 = ScalarParameter() {defaultValue=0.025, parameterName=SpeedDest}
  n7 = Frac(None=Multiply(A=Time(), B=AppendVector(A=@n5, B=@n6)))
  n8 = Multiply(
   A=TextureSample(UVs=Add(A=@n4, B=@n7)) {texture=/Figment_S06/Effects/Fort_Effects/Textures/Misc/T_Tiling_Noise_01}.R
   B=ScalarParameter() {defaultValue=0.6702, parameterName=Distort}
  )
  n9 = Add(A=@n3, B=@n8)
  n10 = Multiply(
   A=LinearInterpolate(
    A=VectorParameter() {parameterName=baseColor, defaultValue=<rgba 0.0834 0.093 0.1771 0>}.RGB
    B=VectorParameter() {parameterName=baseColor_1, defaultValue=<rgba 1.2 1.2 1.2 0>}.RGB
    Alpha=TextureSample(UVs=@n9) {samplerType=SAMPLERTYPE_Masks, texture=/Game/Athena/SafeZone/Textures/StormCloudtubes_M}.G
   )
   B=ScalarParameter() {defaultValue=0.4238, parameterName=Brightness}
  )
  n11 = Cosine(None=ComponentMask(None=TextureCoordinate()) {g=true})
  n12 = ComponentMask(None=TextureCoordinate()) {g=true}
  n13 = LinearInterpolate(
   A=ScalarParameter() {defaultValue=0.1048, parameterName=C_1}
   B=ScalarParameter() {defaultValue=94.2971, parameterName=C_2}
   Alpha=Distance(A=@n12, B=Constant() {r=0.5})
  )
  n14 = Multiply(A=Clamp(None=OneMinus(None=@n13)), B=ScalarParameter() {defaultValue=0.1813, parameterName=FalloffBrightness_2})
  n15 = Divide(
   A=Distance(A=ComponentMask(None=@n2.World Position) {g=true, r=true}, B=Constant2Vector())
   B=ScalarParameter() {defaultValue=3387789824, parameterName=DistanceFadeBlendToWhite}
  )
  n16 = ScalarParameter() {defaultValue=142012528, parameterName=Max}
  n17 = Multiply(
   A=Power(Base=Clamp(None=@n15, Max=@n16), Exp=ScalarParameter() {defaultValue=0.0658, parameterName=FalloffExponent})
   B=ScalarParameter() {defaultValue=0.6312, parameterName=FalloffBrightness}
  )
  Add(
   A=LinearInterpolate(A=@n10, B=VectorParameter() {parameterName=baseColor_2, defaultValue=<rgba 0.1636 0.2083 0.1979 0>}.RGB, Alpha=Clamp(None=@n11))
   B=Multiply(A=Add(A=@n14, B=@n17), B=VectorParameter() {parameterName=TintFalloff, defaultValue=<rgba 0.3594 0.5291 1 1>}.RGB)
  )
  Add(A=@n3)
  n18 = TextureCoordinate()
  n19 = ScalarParameter() {defaultValue=16.2014, parameterName=TilingNexus_X}
  n20 = ScalarParameter() {defaultValue=1.5776, parameterName=TilingNexus_Y}
  n21 = Add(
   A=Multiply(A=@n18, B=AppendVector(A=@n19, B=@n20))
   B=AppendVector(A=ScalarParameter() {parameterName=TilingNexus_Offset_X}, B=ScalarParameter() {defaultValue=-0.2612, parameterName=TilingNexus_Offset_Y})
  )
  n22 = ScalarParameter() {defaultValue=-1.3, parameterName=Height}
  n23 = ScalarParameter() {defaultValue=16, parameterName=TilingNexus_X_1}
  n24 = ScalarParameter() {defaultValue=1.4702, parameterName=TilingNexus_Y_1}
  n25 = Add(
   A=Multiply(A=BumpOffset(Coordinate=@n18, Height=@n22), B=AppendVector(A=@n23, B=@n24))
   B=AppendVector(
    A=ScalarParameter() {defaultValue=0.4847, parameterName=TilingNexus_Offset_X_1}
    B=ScalarParameter() {defaultValue=-0.2326, parameterName=TilingNexus_Offset_Y_1}
   )
  )
  n26 = Multiply(
   A=VectorParameter() {parameterName=Emissive, defaultValue=<rgba 0.4792 0.4792 0.4792 1>}.RGB
   B=Add(
    A=TextureSample(UVs=@n21) {texture=/Figment_S06/Athena/Environments/Nexus/Textures/T_ButterflyBackground_D}.RGB
    B=TextureSample(UVs=@n25) {texture=&T_ButterflyBackground_D}.RGB
   )
  )
  n27 = ScalarParameter() {defaultValue=1.5698, parameterName=EmissiveContrast}
  n28 = TextureCoordinate()
  n29 = Multiply(A=ScalarParameter() {defaultValue=6, parameterName=UVDistort1_ScaleX}, B=ComponentMask(None=@n28) {r=true})
  n30 = Multiply(A=ComponentMask(None=@n28) {g=true}, B=ScalarParameter() {defaultValue=2, parameterName=UVDistort1_ScaleY})
  n31 = VectorParameter() {parameterName=UVDistort1_Speed, defaultValue=<rgba 0.1 -0.5 0 0>}
  n32 = TextureSample!(
   UVs=Panner(Coordinate=AppendVector(A=@n29, B=@n30), Speed=ComponentMask(None=@n31.RGB) {g=true, r=true})
   Tex
   Level=ScalarParameter() {parameterName=MipLevel}
   Apply View MipBias
  ) {mipValueMode=TMVM_MipLevel, samplerType=SAMPLERTYPE_Normal, texture=/Game/VisualThreatManager/StormVisuals/Textures/2ndInfectedMapNormal}
  n33 = TextureCoordinate()
  n34 = AppendVector(
   A=Multiply(A=ScalarParameter() {defaultValue=4, parameterName=UVDistort2_ScaleX}, B=ComponentMask(None=@n33) {r=true})
   B=Multiply(A=ComponentMask(None=@n33) {g=true}, B=ScalarParameter() {defaultValue=2, parameterName=UVDistort2_ScaleY})
  )
  n35 = TextureCoordinate() {uTiling=16}
  n36 = ComponentMask(None=VectorParameter() {parameterName=Speed_3, defaultValue=<rgba 0.1 -0.1 0 0>}.RGB) {g=true, r=true}
  n37 = ScalarParameter() {defaultValue=70, parameterName=BumpOffset_Reveal_ScaleX}
  n38 = TextureCoordinate()
  n39 = ComponentMask(None=@n38) {r=true}
  n40 = ComponentMask(None=@n38) {g=true}
  n41 = ScalarParameter() {defaultValue=1, parameterName=BumpOffset_Reveal_ScaleY}
  n42 = TextureSample!(UVs=AppendVector(A=Multiply(A=@n37, B=@n39), B=Multiply(A=@n40, B=@n41)), Tex, Apply View MipBias) {texture=/Figment_S06/Effects/Fort_Effects/Textures/Tile/T_BlasterBeamDistort}
  n43 = LinearInterpolate(
   B=TextureSample(UVs=Panner(Coordinate=@n35, Speed=@n36) {speedY=0.1}) {texture=/Figment_S06/Developers/timelek/Textures/FlowMaps/T_SwirlFlow_02_D}.G
   Alpha=@n42.G
  )
  n44 = VectorParameter() {parameterName=UVDistor2_Speed, defaultValue=<rgba 0.0125 -0.04 0 0>}
  n45 = TextureSample!(
   UVs=Panner(
    Coordinate=BumpOffset(Coordinate=@n34, Height=@n43) {heightRatio=0.08, referencePlane=3}
    Speed=ComponentMask(None=@n44.RGB) {g=true, r=true}
   )
   Tex
   Level=ScalarParameter() {defaultValue=1, parameterName=MipLevel_1}
   Apply View MipBias
  ) {mipValueMode=TMVM_MipLevel, samplerType=SAMPLERTYPE_Normal, texture=/Figment_S06/Developers/timelek/Textures/FlowMaps/T_SwirlFlow_01_N}
  n46 = LinearInterpolate(
   A=Multiply(A=ComponentMask(None=@n32.RGB) {g=true, r=true}, B=ScalarParameter() {defaultValue=0.05, parameterName=UVDistort1})
   B=Multiply(A=ComponentMask(None=@n45.RGB) {g=true, r=true}, B=ScalarParameter() {defaultValue=0.018, parameterName=UVDistort2})
   Alpha=Power(Base=@n42.G, Exp=ScalarParameter() {defaultValue=0.01, parameterName=UVDistort_LerpPow})
  )
  n47 = Add(A=@n46, B=TextureCoordinate())
  n48 = LinearInterpolate(
   A=VectorParameter() {parameterName=EmissiveDark}.RGB
   B=VectorParameter() {parameterName=EmissiveBright, defaultValue=<rgba 22 2.1576 6.7994 0>}.RGB
   Alpha=Power(Base=ComponentMask(None=@n47) {g=true}, Exp=ScalarParameter() {defaultValue=12.5, parameterName=EmissivePow}) {constExponent=10}
  )
  n49 = Clamp(None=@n48)
  Add(
   A=@n10
   B=Add(
    A=Power(Base=@n26, Exp=@n27)
    B=MaterialFunctionCall(LDR Input=@n49, HDR Tint=@n48) {materialFunction=&DeriveHDRfromLDR}.RemappedRange
   )
  )
  n50 = MaterialFunctionCall(Vector or UVs=Divide(A=@n3, B=ScalarParameter() {defaultValue=31072398, parameterName=Scale_1})) {materialFunction=&VectorToRadialValue}
  n51 = Constant() {r=1}
  n52 = Power(Base=ComponentMask(None=@n50.Radial Coordinates) {g=true}, Exp=ScalarParameter() {defaultValue=1, parameterName=FallofTwist})
  n53 = Multiply(A=Add(A=@n52, B=ScalarParameter() {parameterName=Scale_Twist_Bias}), B=ScalarParameter() {defaultValue=0.399, parameterName=Scale_Twist_Y})
  n54 = Multiply(A=@n50.Radial Coordinates, B=AppendVector(A=ScalarParameter() {defaultValue=2, parameterName=Scale_Twist_X}, B=Divide(A=@n51, B=@n53)))
  n55 = Constant()
  n56 = ComponentMask(None=@n54) {r=true}
  n57 = Time()
  n58 = AppendVector(A=ScalarParameter() {defaultValue=-0.01, parameterName=Speed_x}, B=ScalarParameter() {defaultValue=0.01, parameterName=Speed_y})
  Add(A=Add(A=@n54, B=AppendVector(A=@n55, B=@n56)), B=Frac(None=Multiply(A=@n57, B=@n58)))
  Clamp(None=OneMinus(None=@n52))
  ComponentMask(None=WorldPosition() {worldPositionShaderOffset=WPT_ExcludeAllShaderOffsets}.XYZ) {g=true, r=true}
  Constant2Vector()
  MaterialFunctionCall!(WPO) {materialFunction=/Game/Athena/Environments/Nexus/Materials/MF_SpawnInBetween}
  n59 = CameraPositionWS()
  n60 = ObjectPositionWS()
  Multiply(
   A=Multiply(A=Subtract(A=@n59, B=@n60), B=ScalarParameter() {defaultValue=0.7, parameterName=Paralaxing})
   B=Constant3Vector() {constant=<rgba 1 1 0 0>}
  )
  n61 = TextureCoordinate()
  n62 = Multiply(A=ScalarParameter() {defaultValue=1, parameterName=BaseTexture1_ScaleX}, B=ComponentMask(None=@n61) {r=true})
  n63 = Multiply(A=ComponentMask(None=@n61) {g=true}, B=ScalarParameter() {defaultValue=0.1, parameterName=BaseTexture1_ScaleY})
  n64 = Panner(
   Coordinate=Add(A=AppendVector(A=@n62, B=@n63), B=@n46)
   Speed=ComponentMask(None=VectorParameter() {parameterName=BaseTexture_Speed}.RGB) {g=true, r=true}
  ) {speedY=-5}
  n65 = TextureCoordinate()
  n66 = Multiply(A=ScalarParameter() {defaultValue=2, parameterName=BaseTexture1_ScaleX_1}, B=ComponentMask(None=@n65) {r=true})
  n67 = Multiply(A=ComponentMask(None=@n65) {g=true}, B=ScalarParameter() {defaultValue=2, parameterName=BaseTexture1_ScaleY_1})
  n68 = Panner(
   Coordinate=Add(A=AppendVector(A=@n66, B=@n67), B=@n46)
   Speed=ComponentMask(None=VectorParameter() {parameterName=BaseTexture_Speed_1}.RGB) {g=true, r=true}
  ) {speedY=-5}
  Multiply(
   A=LinearInterpolate(
    A=TextureSample!(UVs=@n64, Tex, Apply View MipBias) {texture=&T_Tiling_Noise_01}.G
    B=TextureSample(UVs=@n68) {texture=/Figment_S06/Effects/Fort_Effects/Textures/Tile/T_FireNoise_01}.G
    Alpha=ScalarParameter() {defaultValue=0.1, parameterName=LerpDetail}
   )
   B=ScalarParameter() {defaultValue=1, parameterName=Roughness}
  )
  Multiply(A=Multiply(A=@n12, B=OneMinus(None=@n12))) {constB=2}
  n69 = WorldPosition()
  Multiply(A=@n6, B=Subtract(A=ComponentMask(None=@n69.XYZ) {g=true, r=true}, B=Constant2Vector()))
  ScalarParameter() {parameterName=met}
  ScalarParameter() {parameterName=PojectionDistance}
  TextureSample(UVs=@n8) {texture=&T_FireNoise_01}
  n70 = TextureCoordinate()
  TextureSample!(UVs=@n54, Tex, "DDX(UVs)"=DDX(Value=@n70), "DDY(UVs)"=DDY(Value=@n70), Apply View MipBias) {mipValueMode=TMVM_Derivative, samplerType=SAMPLERTYPE_Masks, texture=&StormCloudtubes_M}
  WorldPosition()
 ]
}
