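# Two-sided Default Lit material for the Nexus Zero Point, building a colored glow that grades from the actor's center outward.
# World distance from actor position, normalized by object Z scale, drives a color gradient, center glow, outer glow and time-animated sine band.
# These sum with a Fresnel rim and a chromatic smoke texture that uses a second Fresnel as UVs.
# MF_SpawnInBetween supplies world position offset.
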Material /Figment_S06/Athena/Environments/Nexus/Materials/M_ZeroPoint_CenterToEdge
graph {
 properties {pixelDepthOffsetMode=PDOM_Legacy, twoSided=true, usedShadingModels=Default Lit}
 nodes [
  n1 = ScalarParameter() {defaultValue=1.0389, parameterName=CromaticExponent}
  n2 = Fresnel(ExponentIn=ScalarParameter() {defaultValue=3.9666, parameterName=FresnelEXP}) {baseReflectFraction=0}
  n3 = VectorParameter() {parameterName=DistantGlow_1, defaultValue=<rgba 0.6548 0.9701 1 1>}
  n4 = Distance(A=WorldPosition().XYZ, B=ActorPositionWS())
  n5 = Multiply(
   A=ScalarParameter() {defaultValue=2599.2898, parameterName=DistanceBlend_2}
   B=MaterialFunctionCall() {materialFunction=/Engine/Functions/Engine_MaterialFunctions02/WorldPositionOffset/ObjectScale}.Scale Z
  )
  n6 = Multiply(A=Clamp(None=Divide(A=@n4, B=@n5)), B=@n3.RGB)
  n7 = VectorParameter() {parameterName=Color, defaultValue=<rgba 3 0.1412 0.8968 0>}
  n8 = VectorParameter() {parameterName=Color_1, defaultValue=<rgba 0 0.1 0.1 0>}
  n9 = ScalarParameter() {defaultValue=280.2478, parameterName=DistanceBlend}
  n10 = MaterialFunctionCall() {materialFunction=&ObjectScale}
  n11 = Clamp(None=Divide(A=@n4, B=Multiply(A=@n9, B=@n10.Scale Z)))
  n12 = VectorParameter() {parameterName=Color_2, defaultValue=<rgba 0.6457 1.9801 2 0>}
  n13 = Divide(
   A=@n4
   B=Multiply(
    A=ScalarParameter() {defaultValue=201.1141, parameterName=DistanceBlend_1}
    B=MaterialFunctionCall() {materialFunction=&ObjectScale}.Scale Z
   )
  )
  n14 = Frac(None=Multiply(A=Time(), B=ScalarParameter() {defaultValue=0.1429, parameterName=Speed}))
  n15 = Clamp(None=Sine(None=Add(A=@n13, B=@n14)))
  n16 = ScalarParameter() {defaultValue=136.8187, parameterName=DistanceBlend_3}
  n17 = MaterialFunctionCall() {materialFunction=&ObjectScale}
  n18 = Clamp(None=Divide(A=@n4, B=Multiply(A=@n16, B=@n17.Scale Z)))
  n19 = Add(
   A=Add(A=LinearInterpolate(A=@n7.RGB, B=@n8.RGB, Alpha=@n11), B=Multiply(A=@n12.RGB, B=@n15))
   B=Multiply(A=OneMinus(None=@n18), B=VectorParameter() {parameterName=CenterGlow, defaultValue=<rgba 0.6548 0.9701 1 1>}.RGB)
  )
  Add(
   A=TextureSample(UVs=Fresnel(ExponentIn=@n1) {baseReflectFraction=0}) {texture=/Game/Effects/Fort_Effects/Textures/Tile/T_Inky_Smoke_Tile_Chromatic}.RGB
   B=Add(A=Multiply(A=@n2, B=@n3.RGB), B=Add(A=@n6, B=@n19))
  )
  MaterialFunctionCall!(WPO) {materialFunction=/Game/Athena/Environments/Nexus/Materials/MF_SpawnInBetween}
 ]
}
