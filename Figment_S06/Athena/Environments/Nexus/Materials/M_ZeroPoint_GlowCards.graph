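# Additive unlit, two-sided material for Zero Point glow cards in the Nexus environment, drawn as camera-facing flare sprites.
# A cross-flare texture's green channel blends two HDR colors and feeds an alpha multiplier.
# Sprite size blends between near and far values by camera distance.
# Three time-driven sines, offset by Pivot Painter per-element randoms, add object-scaled drift; the combined world position offset passes through MF_SpawnInBetween.
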Material /Figment_S06/Athena/Environments/Nexus/Materials/M_ZeroPoint_GlowCards
graph {
 properties {
  bUseTranslucencyVertexFog=false
  blendMode=BLEND_Additive
  pixelDepthOffsetMode=PDOM_Legacy
  shadingModel=MSM_Unlit
  twoSided=true
  usedShadingModels=Unlit
 }
 nodes [
  n1 = TextureSample() {texture=/Game/Effects/Fort_Effects/Textures/Flares/T_Cross_Flare_ColorsRGB}
  n2 = LinearInterpolate(
   A=VectorParameter() {parameterName=Color0, defaultValue=<rgba 0.0417 0 0 0>}.RGB
   B=VectorParameter() {parameterName=Color1, defaultValue=<rgba 12 1.1884 3.1093 0>}.RGB
   Alpha=@n1.G
  )
  MaterialFunctionCall(LDR Input=Clamp(None=@n2), HDR Tint=@n2) {materialFunction=&DeriveHDRfromLDR}
  n3 = MaterialFunctionCall() {materialFunction=/Engine/Functions/Engine_MaterialFunctions02/WorldPositionOffset/ObjectScale}
  n4 = Multiply(A=ScalarParameter() {defaultValue=8.2791, parameterName=WPOoffset}, B=@n3.Scale X)
  n5 = MaterialFunctionCall() {materialFunction=/Engine/Functions/Engine_MaterialFunctions02/PivotPainter/PivotPainter_PerObjectData}
  n6 = ScalarParameter() {defaultValue=6.8846, parameterName=Speed_Offset}
  n7 = Multiply(A=Add(A=Multiply(A=@n5.Random Value Per Element, B=@n6), B=Time()), B=ScalarParameter() {defaultValue=0.0242, parameterName=Speed_Master})
  n8 = Multiply(A=@n7, B=ScalarParameter() {defaultValue=1.561, parameterName=Speed})
  n9 = Multiply(A=@n7, B=ScalarParameter() {defaultValue=0.8813, parameterName=Speed_1})
  n10 = ScalarParameter() {defaultValue=1.2407, parameterName=Speed_2}
  n11 = AppendVector(A=AppendVector(A=Sine(None=@n8), B=Sine(None=@n9)), B=Sine(None=Multiply(A=@n7, B=@n10)))
  n12 = WorldPosition()
  n13 = LinearInterpolate(
   A=ScalarParameter() {defaultValue=2.5, parameterName=SpriteScaleNear}
   B=ScalarParameter() {defaultValue=118.9776, parameterName=SpriteScaleFar}
   Alpha=MaterialFunctionCall(Distance=ScalarParameter() {defaultValue=6000, parameterName=SoftenDistance}) {materialFunction=&Near_Camera_Fade}.Result
  )
  n14 = Multiply(A=AppendVector(A=@n13, B=@n13), B=@n3.Scale X)
  MaterialFunctionCall!(WPO=Add(
    A=Multiply(A=@n4, B=@n11)
    B=MaterialFunctionCall!(Center Location=@n12.XYZ, 0-1 UVs, XY scale=@n14, OverridePivot, "Normals (Optional)") {materialFunction=&Sprite}.World position offset
   )) {materialFunction=/Game/Athena/Environments/Nexus/Materials/MF_SpawnInBetween}
  Multiply(A=@n1.G, B=ScalarParameter() {defaultValue=1, parameterName=ScaleAlpha})
 ]
}
